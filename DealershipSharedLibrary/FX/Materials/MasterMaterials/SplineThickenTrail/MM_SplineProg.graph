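# Translucent two-sided master material for Niagara mesh-particle trails, with progression, dissolve and emissive driven by dynamic parameters.
# SplineThicken builds width and normal, widened with camera distance; vertex-color-weighted vertical displacement, optional inertia and Perlin noise add to WPO.
# A progression curve texture supplies alpha and emissive curves, alongside base-texture masks, dissolve, angle fade and depth fade.
# Color supports gradients and optional time-of-day fake-sun unlit shading, plus LUT-based HDR emissive with hue shift.
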Material /DealershipSharedLibrary/FX/Materials/MasterMaterials/SplineThickenTrail/MM_SplineProg
graph {
 properties {
  bAutomaticallySetUsageInEditor=false
  bUsedWithNiagaraMeshParticles=true
  blendMode=BLEND_Translucent
  translucencyLightingMode=TLM_VolumetricPerVertexDirectional
  twoSided=true
  usedShadingModels=Default Lit
 }
 nodes [
  MaterialFunctionCall(Emissive=NamedRerouteUsage()) {materialFunction=&MF_FX_EmissiveAdjustments}
  n1 = DynamicParameter() {parameterIndex=1, defaultValue=<rgba 0.35 1 10 10>, paramNames [BaseOffset, BaseScale, Tip Width, Base Width]}
  NamedRerouteDeclaration(None=@n1.Base Width) {name=Base Width, nodeColor=<rgba 0.5 0 1 1>}
  NamedRerouteDeclaration(None=@n1.Tip Width) {name=Tip Width, nodeColor=<rgba 0.5 0 1 1>}
  n2 = ScalarParameter() {defaultValue=1, parameterName=ScaleFactor}
  n3 = NamedRerouteUsage()
  n4 = Distance(A=CameraPositionWS(), B=ParticlePositionWS())
  n5 = ScalarParameter() {defaultValue=200000, group=Spline Thicken, parameterName=Spline - Distance Scale Lenght}
  n6 = LinearInterpolate(
   B=ScalarParameter() {defaultValue=2, group=Spline Thicken, parameterName=Spline - Distance Scale Amount}
   Alpha=Saturate(None=Divide(A=@n4, B=@n5) {constB=100000})
  ) {constA=1, constB=2}
  n7 = TextureCoordinate()
  n8 = OneMinus(None=@n7)
  n9 = ScalarParameter() {parameterName="Flip Tip side (0 or 1)", sliderMax=1}
  n10 = MaterialFunctionCall(
   WidthBase=Multiply(A=@n2, B=Multiply(A=@n3, B=@n6))
   WidthTip=Multiply(A=NamedRerouteUsage(), B=@n2)
   UVs for Thickness=Power(Base=LinearInterpolate(A=@n7, B=@n8, Alpha=@n9) {constAlpha=0}, Exp=ScalarParameter() {defaultValue=1, parameterName=Tip Control}) {constExponent=1}
   Expand U or V UV Channel=StaticBool()
  ) {materialFunction=&SplineThicken}
  n11 = Constant()
  n12 = AppendVector(A=@n11, B=@n11)
  n13 = Multiply(
   A=Multiply(A=ScalarParameter() {defaultValue=1, parameterName=ScaleFactor}, B=NamedRerouteUsage())
   B=LinearInterpolate(Alpha=VertexColor().G) {constA=0.8, constB=1.2}
  ) {constB=-100}
  n14 = VertexColor()
  n15 = Multiply(A=@n14.R, B=@n14.R)
  n16 = Add(A=@n10.WorldPositionOffset, B=Multiply(A=AppendVector(A=@n12, B=@n13), B=Multiply(A=@n15, B=@n15)))
  n17 = ComponentMask(None=TextureCoordinate()) {r=true}
  n18 = ScalarParameter() {defaultValue=3, group=Inertia, parameterName=Dynamic Inertia - Exponent, sortPriority=2}
  n19 = LinearInterpolate(
   B=LinearInterpolate(
    B=VectorParameter() {group=Inertia, parameterName=Dynamic Inertia - Position, defaultValue=<rgba 0 0 0 1>}.RGB
    Alpha=Power(Base=@n17, Exp=@n18) {constExponent=3}
   ) {constAlpha=0, constB=0}
   Alpha=ScalarParameter() {group=Inertia, parameterName=Use Dynamic Inertia, sliderMax=1, sortPriority=0}
  )
  n20 = VertexColor()
  n21 = Multiply(A=Multiply(A=@n20.R, B=NamedRerouteUsage()), B=ScalarParameter() {defaultValue=1, parameterName=ScaleFactor})
  n22 = Add(A=NamedRerouteUsage(), B=ParticleRandom())
  n23 = Add(A=TextureSample(UVs=Add(A=@n21, B=@n22)) {texture=/Game/Effects/Fort_Effects/Textures/Tile/T_PerlinRGB_01_M}.RGB) {constB=-0.5}
  n24 = OneMinus(None=@n20.R)
  n25 = Multiply(A=Multiply(A=Multiply(A=@n23, B=@n24), B=NamedRerouteUsage()), B=ScalarParameter() {defaultValue=1, parameterName=ScaleFactor})
  n26 = Constant()
  NamedRerouteDeclaration(None=Add(A=Add(A=@n16, B=@n19), B=StaticSwitchParameter(True=@n25, False=@n26) {group=WPO, parameterName=UseWPONoise})) {name=WPO, nodeColor=<rgba 0.6448 1 0.0656 1>}
  n27 = VectorParameter() {group=Color, parameterName=Color, sortPriority=1, defaultValue=<rgba 1 1 1 1>}
  n28 = VectorParameter() {group=Color, parameterName=ColorSecondary, sortPriority=1, defaultValue=<rgba 0.0312 0.0312 0.0312 1>}
  n29 = VectorParameter() {parameterName=GradientUV, channelNames {a=TileY, b=TileX, g=OffsetY, r=OffsetX}}
  n30 = Multiply(A=NamedRerouteUsage(), B=AppendVector(A=@n29.TileX, B=@n29.TileX))
  n31 = AppendVector(A=@n29.OffsetX, B=@n29.OffsetX)
  n32 = StaticSwitchParameter(True=LinearInterpolate(A=@n27.RGB, B=@n28.RGB, Alpha=Add(A=@n30, B=@n31)), False=@n27.RGB) {parameterName=UseColorGradient}
  n33 = ParticleColor()
  n34 = ScalarParameter() {defaultValue=1, group=Color, parameterName=Color - Use Particle Color, sortPriority=2}
  n35 = Multiply(
   A=NamedRerouteDeclaration(None=LinearInterpolate(A=@n32, B=@n33.RGB, Alpha=@n34)) {name=Final_Base Color, nodeColor=<rgba 1 0.7266 0 1>}
   B=MaterialFunctionCall(
    Color_Morning=VectorParameter() {group=Color Unlit, parameterName=Color - Morning Multiplier, defaultValue=<rgba 0.9896 0.268 0.4929 0>}.RGB
    Color_Day=VectorParameter() {group=Color Unlit, parameterName=Color - Day Multiplier, defaultValue=<rgba 1 1 1 0>}.RGB
    Color_Evening=VectorParameter() {group=Color Unlit, parameterName=Color - Evening Multiplier, defaultValue=<rgba 0.599 0.2901 0.3358 0>}.RGB
    Color_Night=VectorParameter() {group=Color Unlit, parameterName=Color - Night Multiplier, defaultValue=<rgba 0.3498 0.4864 0.6458 0>}.RGB
   ) {materialFunction=&Color_Time_of_Day}.Result
  )
  n36 = ScalarParameter() {defaultValue=0.7, group=Color Unlit, parameterName=Color - Shadow Multiplier}
  n37 = NamedRerouteUsage()
  n38 = CollectionParameter() {
   collection=&FortniteMaterialParameters
   parameterName=SunlightDirectionalVector
  }
  n39 = ComponentMask(None=@n38) {b=true, g=true, r=true}
  n40 = StaticSwitchParameter(True=NamedRerouteUsage(), False=LinearInterpolate(A=Multiply(A=@n35, B=@n36) {constB=0.74}, B=@n35, Alpha=DotProduct(A=@n37, B=@n39))) {group=Color Unlit, parameterName=IsFlatUnlit_NoTOD}
  n41 = NamedRerouteUsage()
  NamedRerouteDeclaration(None=StaticSwitchParameter(True=Add(A=@n40, B=@n41), False=@n41) {group=Color Unlit, parameterName=IsShaderUnlit, sortPriority=0}) {name=FinalEmissive, nodeColor=<rgba 0.875 0 1 1>}
  n42 = NamedRerouteUsage()
  n43 = Subtract(A=OneMinus(None=ComponentMask(None=@n42) {r=true}))
  n44 = Multiply(A=NamedRerouteUsage())
  n45 = TextureSampleParameter2D(UVs=Saturate(None=Add(A=@n43, B=@n44))) {
   group=Progression
   parameterName=Progression - Curves
   samplerType=SAMPLERTYPE_LinearColor
   texture=/Game/Global/Textures/Default/Blanks/T_Blank_White_ZeroAlpha_LinearColor
   channelNames {a="", b="", g="Emissive ", r=Alpha}
  }
  NamedRerouteDeclaration(None=@n45."Emissive ") {name=EmissiveProgression, nodeColor=<rgba 1 0.6797 0 1>}
  n46 = LinearInterpolate(
   A=ScalarParameter() {defaultValue=800, group=Opacity, parameterName=Opacity - DepthFade - Min}
   B=ScalarParameter() {defaultValue=100, group=Opacity, parameterName=Opacity - DepthFade - Max}
   Alpha=NamedRerouteUsage()
  )
  n47 = ScalarParameter() {defaultValue=1, parameterName=ScaleFactor}
  n48 = ScalarParameter() {defaultValue=1, group=Opacity, parameterName=Opacity - Angle Fade Min, sortPriority=12}
  n49 = ScalarParameter() {group=Opacity, parameterName=Opacity - Angle Fade Max, sortPriority=13}
  n50 = Fresnel(
   ExponentIn=ScalarParameter() {defaultValue=1, group=Opacity, parameterName=Opacity - Angle Fade Exp, sortPriority=11}
   Normal=NamedRerouteUsage()
  ) {baseReflectFraction=0, exponent=1}
  n51 = NamedRerouteUsage()
  n52 = Multiply(
   A=AppendVector(
    A=ScalarParameter() {defaultValue=1, group=Dissolve, parameterName=Dissolve - TextureScale_X, sortPriority=20}
    B=ScalarParameter() {defaultValue=0.2, group=Dissolve, parameterName=Dissolve - TextureScale_Y, sortPriority=21}
   )
   B=NamedRerouteUsage()
  )
  n53 = NamedRerouteUsage()
  n54 = StaticSwitchParameter(True=ParticleRandom(), False=ScalarParameter() {parameterName=DissolveOffsetY}) {defaultValue=true, parameterName=Dissolve Random Particle}
  n55 = TextureSampleParameter2D(UVs=Add(A=Multiply(A=@n51, B=@n52), B=AppendVector(A=@n53, B=@n54))) {group=Dissolve, parameterName=Dissolve Texture, sortPriority=0, texture=/Game/Effects/Fort_Effects/Textures/Tile/T_Inky_Smoke_Tile2}
  n56 = Add(A=ChannelMaskParameter(None=@n55.RGBA) {group=Dissolve, parameterName=Dissolve Texture - Channel, sortPriority=10}.R)
  n57 = Multiply(A=NamedRerouteUsage()) {constB=2}
  n58 = StaticSwitchParameter(True=Power(Base=Subtract(A=@n56, B=@n57), Exp=NamedRerouteUsage()), False=Constant() {r=1}) {group=Dissolve, parameterName=UseDissolve, sortPriority=-10}
  n59 = StaticSwitchParameter(True=Multiply(A=@n58, B=LinearInterpolate(A=@n48, B=@n49, Alpha=@n50)), False=@n58) {group=Opacity, parameterName=UseAngleFade, sortPriority=10}
  n60 = ShadingPathSwitch(Default=@n59, Deferred=DepthFade(Opacity=@n59, FadeDistance=Multiply(A=@n46, B=@n47))) {desc="Defered is pc, Forward is switch, mobile is mobile"}
  n61 = FeatureLevelSwitch(Default=@n59, ES2=@n59, ES3_1=@n59, SM4_REMOVED=@n59, SM5=@n60, SM6=@n60)
  n62 = Constant() {r=1}
  n63 = Multiply(
   A=Multiply(A=@n59, B=StaticSwitchParameter(True=@n61, False=@n62) {group=Opacity, parameterName=UseDephtFade, sortPriority=10})
   B=NamedRerouteUsage()
  )
  n64 = NamedRerouteUsage()
  NamedRerouteDeclaration(None=Multiply(A=Multiply(A=@n63, B=@n64), B=ParticleColor().A)) {name=Opactiy Final, nodeColor=<rgba 0 1 0.5547 1>}
  n65 = DynamicParameter() {defaultValue=<rgba 0.5 1 1 1>, paramNames [Progression, Gravity Amount, Emissive, Emissive Exp]}
  NamedRerouteDeclaration(None=@n65.Progression) {name=Progression_Dyn, nodeColor=<rgba 0.5 0 1 1>}
  NamedRerouteDeclaration(None=@n1.BaseOffset) {name=Offset_Y, nodeColor=<rgba 0.5 0 1 1>}
  NamedRerouteDeclaration(None=@n1.BaseScale) {name=Scale_Y, nodeColor=<rgba 0.5 0 1 1>}
  NamedRerouteDeclaration(None=@n10.Normal) {name=SplineNormal, nodeColor=<rgba 0.5 0.5 1 1>}
  n66 = Sine(None=ComponentMask(None=TextureCoordinate() {uTiling=0.5, vTiling=0.5}) {g=true})
  n67 = Power(Base=Multiply(A=@n66, B=@n66), Exp=ScalarParameter() {defaultValue=1, parameterName=Mask Power})
  n68 = NamedRerouteDeclaration(None=TextureCoordinate()) {name=UVs, nodeColor=<rgba 1 0 0.6328 1>}
  n69 = OneMinus(None=ComponentMask(None=@n68) {r=true})
  n70 = StaticSwitchParameter(True=ScalarParameter() {defaultValue=1, parameterName=Base - TextureScale_X}, False=Constant() {r=1}) {
   desc="Applies gradient through width of base texture. Toggle on when using tile textures (instead of trail textures)."
   parameterName=BaseTextureWidthMask
  }
  n71 = AppendVector(A=NamedRerouteUsage(), B=@n70)
  n72 = Add(A=NamedRerouteUsage(), B=ParticleRandom())
  n73 = Constant()
  n74 = TextureSampleParameter2D(UVs=Add(A=Multiply(A=@n68, B=@n71), B=AppendVector(A=@n72, B=@n73))) {group=Base Texture, parameterName=Base Texture, sortPriority=0, texture=/Game/Effects/Fort_Effects/Textures/Smoke/T_Puffy_Smoke_Trail_01}
  n75 = ChannelMaskParameter(None=@n74.RGBA) {group=Base Texture, parameterName="Base Texture  - Channel", sortPriority=10}
  n76 = Multiply(A=@n75.R, B=ScalarParameter() {defaultValue=1, group=Base Texture, parameterName=Base Texture - Multply, sortPriority=21})
  n77 = ScalarParameter() {defaultValue=1, group=Base Texture, parameterName=Base Texture - Power, sortPriority=22}
  n78 = Multiply(
   A=Saturate(None=Multiply(A=@n69) {constB=25})
   B=StaticSwitchParameter(True=Power(Base=@n76, Exp=@n77), False=@n75.R) {group=Base Texture, parameterName=Use Base Texture Adjustments, sortPriority=20}
  )
  n79 = StaticSwitchParameter(True=Multiply(A=Saturate(None=@n67), B=@n78), False=@n78) {
   desc="Applies gradient through width of base texture. Toggle on when using tile textures (instead of trail textures)."
   parameterName=BaseTextureWidthMask
  }
  NamedRerouteDeclaration(None=@n79) {name=BaseTexture, nodeColor=<rgba 0.0242 0.3813 1 1>}
  n80 = DynamicParameter() {parameterIndex=2, defaultValue=<rgba 1 1 0 0>, paramNames [DissolveOffset, DissolveScale, DissolveContrast, Dissolve]}
  NamedRerouteDeclaration(None=@n80.Dissolve) {name=Dissolve_Dyn, nodeColor=<rgba 0.5 0 1 1>}
  NamedRerouteDeclaration(None=@n80.DissolveContrast) {name=DissolveContrast, nodeColor=<rgba 0.5 0 1 1>}
  NamedRerouteDeclaration(None=@n80.DissolveScale) {name=DissolveScale, nodeColor=<rgba 0.5 0 1 1>}
  NamedRerouteDeclaration(None=@n80.DissolveOffset) {name=DissolveOffset, nodeColor=<rgba 0.5 0 1 1>}
  NamedRerouteDeclaration(None=@n65.Emissive) {name=Emissive_Dyn, nodeColor=<rgba 0.5 0 1 1>}
  NamedRerouteDeclaration(None=@n65.Emissive Exp) {name=Emissive Exp, nodeColor=<rgba 0.5 0 1 1>}
  NamedRerouteDeclaration(None=@n45.Alpha) {name=AlphaProgression, nodeColor=<rgba 0.2422 1 0 1>}
  n81 = DynamicParameter() {parameterIndex=3, defaultValue=<rgba 0.35 1 1 0>, paramNames [WPONoiseOffset, WPONoiseScale, WPONoiseAmount, bIsPainted]}
  NamedRerouteDeclaration(None=@n81.WPONoiseAmount) {name=WPONoiseAmount, nodeColor=<rgba 0.5 0 1 1>}
  NamedRerouteDeclaration(None=@n81.WPONoiseScale) {name=WPONoiseScale, nodeColor=<rgba 0.5 0 1 1>}
  NamedRerouteDeclaration(None=@n81.WPONoiseOffset) {name=WPONoiseOffset, nodeColor=<rgba 0.5 0 1 1>}
  NamedRerouteDeclaration(None=@n81.bIsPainted) {name=IsPainted, nodeColor=<rgba 0.5 0 1 1>}
  n82 = TextureCoordinate() {uTiling=0.5, vTiling=2}
  n83 = AppendVector(
   A=ScalarParameter() {defaultValue=1, group=Normal, parameterName=NormalTileX}
   B=ScalarParameter() {defaultValue=1, group=Normal, parameterName=NormalTileY}
  )
  n84 = ScalarParameter() {defaultValue=1, group=Normal, parameterName=NormalOffsetX}
  n85 = ScalarParameter() {defaultValue=1, group=Normal, parameterName=NormalOffsetY}
  n86 = TextureSampleParameter2D(UVs=Add(A=Multiply(A=@n82, B=@n83), B=AppendVector(A=@n84, B=@n85))) {
   group=Normal
   parameterName=Normal Texture
   samplerType=SAMPLERTYPE_Normal
   sortPriority=0
   texture=/Game/Effects/Fort_Effects/Textures/Tile/T_Wind_N
  }
  n87 = MaterialFunctionCall(Normal=@n86.RGB, Flatness=ScalarParameter() {defaultValue=1, group=Normal, parameterName=Normal - Flatten}) {materialFunction=&FlattenNormal}
  NamedRerouteDeclaration(None=MaterialFunctionCall(BaseNormal=NamedRerouteUsage(), AdditionalNormal=Transform(None=@n87.Result) {transformType=TRANSFORM_Instance}) {materialFunction=&BlendAngleCorrectedNormals}.Result) {name=FinalNormal, nodeColor=<rgba 0.5 0.5 1 1>}
  n88 = NamedRerouteUsage()
  n89 = NamedRerouteUsage()
  n90 = TextureSampleParameter2D(UVs=Power(Base=Multiply(A=@n88, B=@n89), Exp=NamedRerouteUsage())) {
   group=Emissive
   parameterName=Emissive - LUT
   sortPriority=1
   texture=/Game/Global/Materials/Character_Parent_2023/FlipbookFace_Parent_2023/Textures/T_black_emptyalpha
  }
  n91 = Multiply(A=NamedRerouteUsage(), B=@n90.RGB)
  MaterialExpressionReroute_9 = Reroute(None=@n90.RGB)
  n92 = Switch!(SwitchValue=NamedRerouteUsage(), Default=Multiply(A=ParticleColor().RGB, B=Desaturation(None=@n90.RGB)), None=@MaterialExpressionReroute_9) {inputs [{
     inputName=None
     input {
      expression=&self:MaterialExpressionReroute_9
      inputName=None
      mask=0
      maskA=0
      maskB=0
      maskG=0
      maskR=0
      outputIndex=0
     }
    }]}
  n93 = MaterialFunctionCall(
   Hue Shift Percentage=ScalarParameter() {group=Emissive, parameterName=Emissive - HueShift, sortPriority=10}
   Texture=Desaturation(
    None=StaticSwitchParameter(True=@n91, False=@n92) {group=Emissive, parameterName=Multiply HDR Emissive by Base Color, sortPriority=2}
    Fraction=ScalarParameter() {group=Emissive, parameterName=Emissive_Desaturation, sortPriority=2}
   )
  ) {materialFunction=&HueShift}
  n94 = Saturate(None=@n93.Result)
  n95 = Multiply(A=ScalarParameter() {defaultValue=100, group=Emissive, parameterName=Emissive - HDR Intensity, sortPriority=6}, B=NamedRerouteUsage()) {constB=0}
  NamedRerouteDeclaration(None=StaticSwitchParameter(
    True=MaterialFunctionCall(LDR Input=@n94, HDR Tint=@n93.Result, HDR Intensity Multiplier=@n95) {materialFunction=&DeriveHDRfromLDR}.RemappedRange
    False=Constant()
   ) {group=Emissive, parameterName=UseEmissive, sortPriority=0}) {name=HDR Emissive, nodeColor=<rgba 1 0.1098 0.1289 1>}
  NamedRerouteDeclaration(None=@n65.Gravity Amount) {name=Gravity Amount, nodeColor=<rgba 0.5 0 1 1>}
  NamedRerouteUsage()
  NamedRerouteUsage()
  NamedRerouteUsage()
  NamedRerouteUsage()
 ]
}
