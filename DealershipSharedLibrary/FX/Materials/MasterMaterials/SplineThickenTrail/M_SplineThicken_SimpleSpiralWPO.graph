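# Masked, unlit, two-sided master material for spline-thickened trails on Niagara mesh particles, adding spiral wobble through vertex offset.
# SplineThicken takes Spline Width Base for both ends; dynamic-parameter gravity adds vertical sag weighted by vertex red to the fourth power, plus a uniform push.
# Sine and cosine of tiled U build a transformed wobble, fading along U.
# Particle color feeds adjusted emissive; alpha minus U gives an erosion mask.
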Material /DealershipSharedLibrary/FX/Materials/MasterMaterials/SplineThickenTrail/M_SplineThicken_SimpleSpiralWPO
graph {
 properties {bUsedWithNiagaraMeshParticles=true, blendMode=BLEND_Masked, shadingModel=MSM_Unlit, twoSided=true, usedShadingModels=Unlit}
 nodes [
  n1 = DynamicParameter() {paramNames [Spline Width Base, Spline Width Tip, Gravity, GravityForce]}
  n2 = MaterialFunctionCall(WidthBase=@n1.Spline Width Base, WidthTip=@n1.Spline Width Base, Expand U or V UV Channel=StaticBool()) {materialFunction=&SplineThicken}
  n3 = LinearInterpolate(Alpha=VertexColor().G) {constA=-125, constB=-250}
  n4 = VertexColor()
  n5 = Multiply(A=@n4.R, B=@n4.R)
  n6 = Multiply(
   A=MaterialFunctionCall(Z=Multiply(A=@n1.Gravity, B=@n3) {constB=-100}) {materialFunction=&MakeFloat3}.Result
   B=Multiply(A=@n5, B=@n5)
  )
  n7 = VectorParameter() {group=WPO, parameterName=WPO Params, channelNames {a="", b=Strength, g="", r=Tiling}}
  n8 = ComponentMask(None=Multiply(A=TextureCoordinate(), B=@n7.Tiling)) {r=true}
  n9 = Cosine(None=@n8)
  n10 = Sine(None=@n8)
  n11 = Transform(None=Multiply(
    A=MaterialFunctionCall(X=@n9, Y=@n10, Z=@n9) {materialFunction=&MakeFloat3}.Result
    B=@n7.Strength
   ))
  n12 = OneMinus(None=ComponentMask(None=TextureCoordinate()) {r=true})
  Add(
   A=Add(
    A=Add(A=@n2.WorldPositionOffset, B=@n6)
    B=MaterialFunctionCall(Z=@n1.GravityForce) {materialFunction=&MakeFloat3}.Result
   )
   B=Transform(None=Multiply(A=@n11, B=@n12)) {transformSourceType=TRANSFORMSOURCE_World, transformType=TRANSFORM_Local}
  )
  n13 = ParticleColor()
  MaterialFunctionCall(Emissive=@n13.RGB) {materialFunction=&MF_FX_EmissiveAdjustments}
  Subtract(A=@n13.A, B=ComponentMask(None=TextureCoordinate()) {r=true})
 ]
}
